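# Translucent, unlit, two-sided Niagara ribbon trail material for the Phleyms set.
# A panning RGB mask, with switchable tiling, drives emissive color scaled by particle color and dynamic Highlights and Brightness parameters.
# U-based fades shape the blue-channel opacity mask, scaled by particle alpha, and a separate Distortion-scaled term.
# An optional WPO switch adds time-animated sine and cosine wobble along UV1, ramped along UV0 and tuned by scalar parameters.
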Material /DealershipSharedLibrary/Assets/Sets/Phleyms/Materials/M_DLR_Trail_01
graph {
 properties {
  bUsedWithNiagaraRibbons=true
  blendMode=BLEND_Translucent
  pixelDepthOffsetMode=PDOM_Legacy
  shadingModel=MSM_Unlit
  twoSided=true
  usedShadingModels=Unlit
 }
 nodes [
  n1 = TextureCoordinate() {uTiling=2}
  n2 = TextureCoordinate() {uTiling=1.5}
  n3 = TextureSample(UVs=StaticSwitchParameter(True=Panner(Coordinate=@n1) {speedX=-2}, False=Panner(Coordinate=@n2) {speedX=-1.5}) {defaultValue=true, parameterName=Tile2x}) {texture=/DealershipSharedLibrary/Assets/Sets/Phleyms/Textures/T_DLR_Phleyms_Trail_Mask_RGB}
  Add(A=@n3.RGBA, B=Constant3Vector() {constant=<rgba 0 0 1 0>})
  Constant3Vector()
  n4 = DynamicParameter() {paramNames [Highlights,Brightness,Distortion,InnerDetail]}
  n5 = Desaturation(None=Add(A=Multiply(A=@n4.Highlights, B=@n3.R), B=@n3.RGBA))
  n6 = ParticleColor()
  MaterialFunctionCall(Emissive=Multiply(A=Multiply(A=@n5, B=@n4.Brightness), B=@n6.RGB)) {materialFunction=&MF_FX_EmissiveAdjustments}
  n7 = ComponentMask(None=TextureCoordinate()) {r=true}
  n8 = OneMinus(None=@n7)
  n9 = Multiply(A=Saturate(None=ConstantBiasScale(None=@n8) {bias=-0.01, scale=2}), B=Saturate(None=ConstantBiasScale(None=@n7) {bias=-0.025, scale=20}))
  Multiply(A=Multiply(A=Add(A=@n3.R, B=@n3.B), B=@n9), B=@n4.Distortion)
  n10 = Constant() {r=8}
  n11 = Saturate(None=ConstantBiasScale(None=Multiply(A=@n3.B, B=@n10)) {bias=-0.01, scale=1})
  Saturate(None=Multiply(A=@n6.A, B=Multiply(A=@n11, B=@n9)))
  n12 = ConstantBiasScale(None=ComponentMask(None=TextureCoordinate()) {r=true}) {bias=-0.01, scale=2}
  n13 = Time()
  n14 = TextureCoordinate() {coordinateIndex=1, uTiling=3}
  n15 = ScalarParameter() {defaultValue=1, group=WPO, parameterName=tile}
  n16 = Add(
   A=Multiply(A=ScalarParameter() {defaultValue=1, group=WPO, parameterName=Speed}, B=OneMinus(None=@n13))
   B=ComponentMask(None=Multiply(A=@n14, B=@n15)) {r=true}
  )
  n17 = Add(A=Cosine(None=@n16), B=Cosine(None=@n16) {period=2})
  n18 = AppendVector(A=@n17, B=Add(A=Sine(None=@n16), B=Sine(None=@n16) {period=2}))
  n19 = Multiply(
   A=Multiply(A=AppendVector(A=@n18, B=@n17), B=ScalarParameter() {defaultValue=1, group=WPO, parameterName=Amount})
   B=ScalarParameter() {defaultValue=1, group=WPO, parameterName=Intensity}
  )
  StaticSwitchParameter(True=Multiply(A=Saturate(None=@n12), B=Transform(None=@n19)), False=Constant()) {group=WPO, parameterName=WPO}
 ]
}
